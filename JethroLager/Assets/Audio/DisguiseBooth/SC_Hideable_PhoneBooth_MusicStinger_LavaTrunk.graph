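# Sound cue for the LavaTrunk music stinger of the hideable disguise phone booth.
# A fixed one-second Delay node feeds a single Wave Player that plays the stinger wave.
# The cue uses the gadgets and items sound class with a slight volume boost and a very large max distance.
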SoundCue /JethroLager/Assets/Audio/DisguiseBooth/SC_Hideable_PhoneBooth_MusicStinger_LavaTrunk
root=Delay_0
nodes [
 {id=Delay_0, children [WavePlayer_2], properties {delayMax=1, delayMin=1}}
 {id=WavePlayer_2, properties {soundWaveAssetPtr=&/MX_S35_PhoneBooth_LavaTrunk_STNG}}
]
properties {
 maxDistance=2097152
 soundClassObject=&Fort_Gadgets_Items
 volumeMultiplier=1.0714
 assetUserData [&self:AudioAssetUserData_0]
}
