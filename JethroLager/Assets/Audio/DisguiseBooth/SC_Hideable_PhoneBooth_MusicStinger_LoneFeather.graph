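# Sound cue for the LoneFeather music stinger of a hideable phone booth disguise prop.
# A single wave player feeds the output with the PhoneBooth LoneFeather stinger wave.
# It routes through the Fort_Gadgets_Items sound class with a slight volume boost and a very large max distance.
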SoundCue /JethroLager/Assets/Audio/DisguiseBooth/SC_Hideable_PhoneBooth_MusicStinger_LoneFeather
root=WavePlayer_2
nodes [{id=WavePlayer_2, properties {soundWaveAssetPtr=&/MX_S35_PhoneBooth_LoneFeather_STNG}}]
properties {maxDistance=2097152, soundClassObject=&Fort_Gadgets_Items, volumeMultiplier=1.0714}
